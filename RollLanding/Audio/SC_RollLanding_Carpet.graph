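# Sound cue for a roll landing on carpet, built from wood-solid bodyfall waves.
# A PlayerRelation switch picks first-person, friendly or enemy mixes, each with its own volume and sound class; enemy is also the first input.
# First-person randomizes the standard waves, while both third-person mixes share an Elevation crossfade across below, level and above attenuations.
# Below and above both draw on the "Above" wave set, and concurrency is capped at two.
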SoundCue /RollLanding/Audio/SC_RollLanding_Carpet
root=Modulator_1
nodes [
 {id=Attenuation_0, children [Random_0]}
 {
  id=Attenuation_1
  children [Random_2]
  properties {attenuationSettings=/Game/Sounds/Fort_Player/Footsteps/Settings/Fort_Player_FS_3P_Below_Att}
 }
 {id=Attenuation_2, children [Random_0], properties {attenuationSettings=/Game/Sounds/Fort_Player/Footsteps/Settings/Fort_Player_FS_3P_Att}}
 {
  id=Attenuation_3
  children [Random_2]
  properties {attenuationSettings=/Game/Sounds/Fort_Player/Footsteps/Settings/Fort_Player_FS_3P_Above_Att}
 }
 {id=Mixer_0, children [SoundClass_3], properties {inputVolume [2]}}
 {id=Mixer_1, children [SoundClass_5], properties {inputVolume [0.55]}}
 {id=Mixer_2, children [SoundClass_4], properties {inputVolume [0.2]}}
 {id=Modulator_1, children [Switch_0], properties {volumeMax=1}}
 {
  id=ParamCrossFade_1
  children [Attenuation_1, Attenuation_2, Attenuation_3]
  properties {
   paramName=Elevation
   crossFadeInput [
    {fadeInDistanceEnd=1, fadeInDistanceStart=0, fadeOutDistanceEnd=2, fadeOutDistanceStart=1.66, volume=2}
    {fadeInDistanceEnd=2, fadeInDistanceStart=1.66, fadeOutDistanceEnd=2.33, fadeOutDistanceStart=2, volume=1}
    {fadeInDistanceEnd=2.33, fadeInDistanceStart=2, fadeOutDistanceEnd=4, fadeOutDistanceStart=3, volume=1}
   ]
  }
 }
 {
  id=Random_0
  children [WavePlayer_0,WavePlayer_1,WavePlayer_2,WavePlayer_3,WavePlayer_4]
  properties {bShouldExcludeFromBranchCulling=true, weights [1,1,1,1,1]}
 }
 {
  id=Random_2
  children [WavePlayer_13, WavePlayer_14, WavePlayer_15, WavePlayer_16, WavePlayer_17, WavePlayer_18, WavePlayer_19]
  properties {bShouldExcludeFromBranchCulling=true, weights [1,1,1,1,1,1,1]}
 }
 {id=SoundClass_3, children [Attenuation_0], properties {soundClassOverride=&Fort_Player_Foley_1P_Athena}}
 {
  id=SoundClass_4
  children [ParamCrossFade_1]
  properties {soundClassOverride=/Game/Sounds/FortSoundClassesAndModes/Fort_Player_Foley_3P_Friendly_Athena}
 }
 {
  id=SoundClass_5
  children [ParamCrossFade_1]
  properties {soundClassOverride=&Fort_Player_Foley_3P_Enemy_Athena}
 }
 {id=Switch_0, children [Mixer_1,Mixer_0,Mixer_2,Mixer_1], properties {intParameterName=PlayerRelation}}
 {id=WavePlayer_0, properties {soundWaveAssetPtr=/Game/Sounds/Fort_Player/Bodyfalls/Waves/Bodyfall_Wood_Solid_01}}
 {id=WavePlayer_1, properties {soundWaveAssetPtr=/Game/Sounds/Fort_Player/Bodyfalls/Waves/Bodyfall_Wood_Solid_02}}
 {id=WavePlayer_13, properties {soundWaveAssetPtr=/Game/Sounds/Fort_Player/Bodyfalls/Waves/Bodyfall_Wood_Solid_Above_01}}
 {id=WavePlayer_14, properties {soundWaveAssetPtr=/Game/Sounds/Fort_Player/Bodyfalls/Waves/Bodyfall_Wood_Solid_Above_02}}
 {id=WavePlayer_15, properties {soundWaveAssetPtr=/Game/Sounds/Fort_Player/Bodyfalls/Waves/Bodyfall_Wood_Solid_Above_03}}
 {id=WavePlayer_16, properties {soundWaveAssetPtr=/Game/Sounds/Fort_Player/Bodyfalls/Waves/Bodyfall_Wood_Solid_Above_04}}
 {id=WavePlayer_17, properties {soundWaveAssetPtr=/Game/Sounds/Fort_Player/Bodyfalls/Waves/Bodyfall_Wood_Solid_Above_05}}
 {id=WavePlayer_18, properties {soundWaveAssetPtr=/Game/Sounds/Fort_Player/Bodyfalls/Waves/Bodyfall_Wood_Solid_Above_06}}
 {id=WavePlayer_19, properties {soundWaveAssetPtr=/Game/Sounds/Fort_Player/Bodyfalls/Waves/Bodyfall_Wood_Solid_Above_07}}
 {id=WavePlayer_2, properties {soundWaveAssetPtr=/Game/Sounds/Fort_Player/Bodyfalls/Waves/Bodyfall_Wood_Solid_03}}
 {id=WavePlayer_3, properties {soundWaveAssetPtr=/Game/Sounds/Fort_Player/Bodyfalls/Waves/Bodyfall_Wood_Solid_04}}
 {id=WavePlayer_4, properties {soundWaveAssetPtr=/Game/Sounds/Fort_Player/Bodyfalls/Waves/Bodyfall_Wood_Solid_05}}
]
properties {
 attenuationSettings=&/ATT_RollLanding
 bOverrideConcurrency=true
 maxDistance=2097152
 virtualizationMode=Disabled
 volumeMultiplier=1
 concurrencyOverrides {maxCount=2}
}
