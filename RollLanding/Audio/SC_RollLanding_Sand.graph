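# Sound cue for roll landings on sand, playing random bodyfall waves with concurrency capped at two.
# A PlayerRelation switch picks first-person, friendly or enemy branches, each with its own sound class and volume, first-person loudest and friendly quietest.
# Both third-person branches share an Elevation crossfade between below, level and above attenuations.
# The below and above layers reuse Concrete_Above waves; first-person and level layers use sand waves.
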SoundCue /RollLanding/Audio/SC_RollLanding_Sand
root=Modulator_2
nodes [
 {id=Attenuation_5, children [Random_1]}
 {
  id=Attenuation_6
  children [Random_2]
  properties {attenuationSettings=/Game/Sounds/Fort_Player/Footsteps/Settings/Fort_Player_FS_3P_Below_Att}
 }
 {id=Attenuation_7, children [Random_1], properties {attenuationSettings=/Game/Sounds/Fort_Player/Footsteps/Settings/Fort_Player_FS_3P_Att}}
 {
  id=Attenuation_8
  children [Random_2]
  properties {attenuationSettings=/Game/Sounds/Fort_Player/Footsteps/Settings/Fort_Player_FS_3P_Above_Att}
 }
 {id=Mixer_6, children [SoundClass_6], properties {inputVolume [2]}}
 {id=Mixer_7, children [SoundClass_8], properties {inputVolume [0.55]}}
 {id=Mixer_8, children [SoundClass_7], properties {inputVolume [0.2]}}
 {id=Modulator_2, children [Switch_1], properties {volumeMax=1}}
 {
  id=ParamCrossFade_0
  children [Attenuation_6, Attenuation_7, Attenuation_8]
  properties {
   paramName=Elevation
   crossFadeInput [
    {fadeInDistanceEnd=1, fadeInDistanceStart=0, fadeOutDistanceEnd=2, fadeOutDistanceStart=1.66, volume=2}
    {fadeInDistanceEnd=2, fadeInDistanceStart=1.66, fadeOutDistanceEnd=2.33, fadeOutDistanceStart=2, volume=1}
    {fadeInDistanceEnd=2.33, fadeInDistanceStart=2, fadeOutDistanceEnd=4, fadeOutDistanceStart=3, volume=1}
   ]
  }
 }
 {
  id=Random_1
  children [WavePlayer_6, WavePlayer_7, WavePlayer_8, WavePlayer_9, WavePlayer_10, WavePlayer_11]
  properties {bShouldExcludeFromBranchCulling=true, weights [1,1,1,1,1,1]}
 }
 {
  id=Random_2
  children [WavePlayer_25, WavePlayer_26, WavePlayer_27, WavePlayer_28]
  properties {bShouldExcludeFromBranchCulling=true, weights [1,1,1,1]}
 }
 {id=SoundClass_6, children [Attenuation_5], properties {soundClassOverride=&Fort_Player_Foley_1P_Athena}}
 {
  id=SoundClass_7
  children [ParamCrossFade_0]
  properties {soundClassOverride=/Game/Sounds/FortSoundClassesAndModes/Fort_Player_Foley_3P_Friendly_Athena}
 }
 {
  id=SoundClass_8
  children [ParamCrossFade_0]
  properties {soundClassOverride=&Fort_Player_Foley_3P_Enemy_Athena}
 }
 {id=Switch_1, children [Mixer_7,Mixer_6,Mixer_8,Mixer_7], properties {intParameterName=PlayerRelation}}
 {id=WavePlayer_10, properties {soundWaveAssetPtr=/Game/Sounds/Fort_Player/Bodyfalls/Waves/Bodyfall_Sand_05}}
 {id=WavePlayer_11, properties {soundWaveAssetPtr=/Game/Sounds/Fort_Player/Bodyfalls/Waves/Bodyfall_Sand_06}}
 {id=WavePlayer_25, properties {soundWaveAssetPtr=/Game/Sounds/Fort_Player/Bodyfalls/Waves/Bodyfall_Concrete_Above_01}}
 {id=WavePlayer_26, properties {soundWaveAssetPtr=/Game/Sounds/Fort_Player/Bodyfalls/Waves/Bodyfall_Concrete_Above_02}}
 {id=WavePlayer_27, properties {soundWaveAssetPtr=/Game/Sounds/Fort_Player/Bodyfalls/Waves/Bodyfall_Concrete_Above_03}}
 {id=WavePlayer_28, properties {soundWaveAssetPtr=/Game/Sounds/Fort_Player/Bodyfalls/Waves/Bodyfall_Concrete_Above_04}}
 {id=WavePlayer_6, properties {soundWaveAssetPtr=/Game/Sounds/Fort_Player/Bodyfalls/Waves/Bodyfall_Sand_01}}
 {id=WavePlayer_7, properties {soundWaveAssetPtr=/Game/Sounds/Fort_Player/Bodyfalls/Waves/Bodyfall_Sand_02}}
 {id=WavePlayer_8, properties {soundWaveAssetPtr=/Game/Sounds/Fort_Player/Bodyfalls/Waves/Bodyfall_Sand_03}}
 {id=WavePlayer_9, properties {soundWaveAssetPtr=/Game/Sounds/Fort_Player/Bodyfalls/Waves/Bodyfall_Sand_04}}
]
properties {
 attenuationSettings=&/ATT_RollLanding
 bOverrideConcurrency=true
 maxDistance=2097152
 virtualizationMode=Disabled
 volumeMultiplier=1
 concurrencyOverrides {maxCount=2}
}
